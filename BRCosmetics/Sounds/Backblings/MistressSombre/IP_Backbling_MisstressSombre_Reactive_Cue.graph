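# Sound cue for the Mistress Sombre back bling's reactive sound, a crow sound layered with a delayed reactive effect.
# A mixer plays a random one of three crow waves at once and, slightly louder after a short fixed delay, a random one of three reactive waves.
# Each layer gets slight pitch and volume modulation for variation.
# It uses close-range pet attenuation, the back bling sound class and a reactive cosmetic property sheet.
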SoundCue /BRCosmetics/Sounds/Backblings/MistressSombre/IP_Backbling_MisstressSombre_Reactive_Cue
root=Mixer_0
nodes [
 {id=Delay_2, children [Modulator_1], properties {delayMax=0.3, delayMin=0.3}}
 {id=Mixer_0, children [Modulator_0,Delay_2], properties {inputVolume [1,1.2]}}
 {id=Modulator_0, children [Random_0], properties {pitchMax=1.02, pitchMin=0.99, volumeMax=1.02, volumeMin=0.98}}
 {id=Modulator_1, children [Random_1], properties {pitchMax=1.02, pitchMin=0.99, volumeMax=1.02, volumeMin=0.98}}
 {id=Random_0, children [WavePlayer_3,WavePlayer_4,WavePlayer_5], properties {weights [1,1,1]}}
 {id=Random_1, children [WavePlayer_6,WavePlayer_7,WavePlayer_8], properties {weights [1,1,1]}}
 {id=WavePlayer_3, properties {soundWaveAssetPtr=&/IP_Backbling_MistressSombre_Crow_01}}
 {id=WavePlayer_4, properties {soundWaveAssetPtr=&/IP_Backbling_MistressSombre_Crow_02}}
 {id=WavePlayer_5, properties {soundWaveAssetPtr=&/IP_Backbling_MistressSombre_Crow_03}}
 {id=WavePlayer_6, properties {soundWaveAssetPtr=&/IP_Backbling_MistressSombre_Reactive_01}}
 {id=WavePlayer_7, properties {soundWaveAssetPtr=&/IP_Backbling_MistressSombre_Reactive_02}}
 {id=WavePlayer_8, properties {soundWaveAssetPtr=&/IP_Backbling_MistressSombre_Reactive_03}}
]
properties {
 attenuationSettings=&Fort_Pets_Close_Athena_Att
 audioPropertiesSheet=/BRCosmetics/Sounds/Settings/PropertySheets/Reactive/APS_Cosmetic_Reactive_Backbling
 maxDistance=1800.0001
 soundClassObject=/Game/Sounds/FortSoundClassesAndModes/Fort_Backblings
 volumeMultiplier=1
}
