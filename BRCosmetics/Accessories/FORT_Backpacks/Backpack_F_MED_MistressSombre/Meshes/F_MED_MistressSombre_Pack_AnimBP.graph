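# Animation Blueprint for the Mistress Sombre back bling, which plays idle and reaction animations.
# It picks a front-end or gameplay state machine, layers an additive floating-orb animation, then offsets pack and crow bones.
# Briefly pulsed input events drive chest-open and elimination reactions (the latter gated by the modifier's reactivity flag), while previews replay elimination on a delayed notify loop.
# The event graph also computes gimbal, sine-wobble and gravity values, not visibly wired into the pose.
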AnimBlueprint /BRCosmetics/Accessories/FORT_Backpacks/Backpack_F_MED_MistressSombre/Meshes/F_MED_MistressSombre_Pack_AnimBP
target_skeleton=&/F_MED_MistressSombre_Pack_Skeleton
graphs [
 AnimGraph {nodes [
   AnimGraphNode_StateMachine_1 = FE Idle / Elim() {is=title AnimGraphNode_StateMachine}
   AnimGraphNode_StateMachine_2 = Idle / Reactive() {is=title AnimGraphNode_StateMachine}
   n1 = Blend Poses by bool(BlendPose_0=@AnimGraphNode_StateMachine_1, BlendPose_1=@AnimGraphNode_StateMachine_2, bActiveValue=Get bIsInFrontEnd()) {blendType=Linear}
   n2 = "Sequence Player (additive) 'MistressSombre_FloatingOrb_PACK'"() {sequence=/BRCosmetics/Accessories/FORT_Backpacks/Backpack_F_MED_MistressSombre/Animation/MistressSombre_FloatingOrb_PACK}
   n3 = "Transform (Modify) Bone - Bone: pack"(
    ComponentPose=Local To Component(LocalPose=Apply Additive!(Base=@n1, Additive=@n2, bAlphaBoolEnabled:bool=True) -> (Pose) {alphaInputType=Bool, alphaBoolBlend {blendOption=Linear}})
    Translation="2,-12,0"
    Rotation="0,-30,0"
    Scale="0.8,0.8,0.8"
   ) {
    rotationMode=BMM_Additive
    scaleMode=BMM_Replace
    translationMode=BMM_Replace
    alphaBoolBlend {blendOption=Linear}
    boneToModify {boneName=pack}
    rotation {yaw=-30}
    scale=<0.8 0.8 0.8>
    translation=<2 -12>
   }
   n4 = "Transform (Modify) Bone - Bone: crow_ball"(ComponentPose=@n3, Alpha=1, Translation="0,0,0", Rotation="0,0,0", Scale="1.2,1.2,1.2") {scaleMode=BMM_Additive, alphaBoolBlend {blendOption=Linear}, boneToModify {boneName=crow_ball}, scale=<1.2 1.2 1.2>}
   n5 = "Transform (Modify) Bone - Bone: crow_Off_Root_m"(ComponentPose=@n4, Alpha=1, Translation="0,0,0", Rotation="0,0,0", Scale="1,1,1") {
    rotationMode=BMM_Additive
    rotationSpace=BCS_BoneSpace
    alphaBoolBlend {blendOption=Linear}
    alphaScaleBiasClamp {bInterpResult=true}
    boneToModify {boneName=crow_Off_Root_m}
   }
   n6 = "Transform (Modify) Bone - Bone: crow_Off_Root_m"!(ComponentPose=@n5, bAlphaBoolEnabled:bool=True, Translation="0,0,0", Rotation="0,0,0", Scale="1,1,1") -> (Pose) {
    alphaInputType=Bool
    rotationMode=BMM_Additive
    rotationSpace=BCS_BoneSpace
    alphaBoolBlend {blendOption=QuarticInOut, blendOutTime=10}
    alphaScaleBiasClamp {bInterpResult=true}
    boneToModify {boneName=crow_Off_Root_m}
   }
   n7 = "Transform (Modify) Bone - Bone: crow_Off_Root_m"!(ComponentPose=@n6, Alpha:float=1, Translation="5,-5,0", Rotation="0,0,0", Scale="1,1,1") -> (Pose) {
    rotationMode=BMM_Additive
    rotationSpace=BCS_BoneSpace
    translationMode=BMM_Additive
    alphaBoolBlend {blendOption=Linear}
    alphaScaleBiasClamp {bInterpResult=true}
    boneToModify {boneName=crow_Off_Root_m}
    translation=<5 -5>
   }
   n8 = "Transform (Modify) Bone - Bone: pack"(ComponentPose=@n7, Translation="0,-5,0", Rotation="0,0,0") {
    rotationMode=BMM_Additive
    rotationSpace=BCS_BoneSpace
    translationMode=BMM_Additive
    translationSpace=BCS_BoneSpace
    alphaBoolBlend {blendOption=Linear}
    alphaScaleBiasClamp {bInterpResult=true}
    boneToModify {boneName=pack}
    translation {y=-5}
   }
   Output Pose(Result=Component To Local(ComponentPose=@n8)) {layerGroup=None}
  ]}
 FE Idle / Elim {
  entry=Idle
  path=AnimGraph.AnimGraphNode_StateMachine_1.FE Idle / Elim
  ids=id
  nodes [
   AnimStateNode_0 = Idle()
   AnimStateNode_1 = EnemyPlayerElimination() {is=title AnimStateNode, stateEntered {guid=#1, notifyName=FE_ElimEntered}}
   AnimStateTransitionNode_0 = AnimStateTransitionNode()
   AnimStateTransitionNode_1 = AnimStateTransitionNode() {bAutomaticRuleBasedOnSequencePlayerInState=true}
  ]
  links [
   {
    dst=Idle
    rule=AnimGraph.AnimGraphNode_StateMachine_1.FE Idle / Elim.AnimStateTransitionNode_1.Transition
    src=EnemyPlayerElimination
    via=AnimStateTransitionNode_1
   }
   {
    dst=EnemyPlayerElimination
    rule=AnimGraph.AnimGraphNode_StateMachine_1.FE Idle / Elim.AnimStateTransitionNode_0.Transition
    src=Idle
    via=AnimStateTransitionNode_0
   }
  ]
 }
 Idle {
  path=AnimGraph.AnimGraphNode_StateMachine_1.FE Idle / Elim.AnimStateNode_0.Idle
  nodes [Output Animation Pose(Result=Sequence Player 'MistressSombre_Idle_PACK'() {sequence=/BRCosmetics/Accessories/FORT_Backpacks/Backpack_F_MED_MistressSombre/Animation/MistressSombre_Idle_PACK})]
 }
 EnemyPlayerElimination {
  path=AnimGraph.AnimGraphNode_StateMachine_1.FE Idle / Elim.AnimStateNode_1.EnemyPlayerElimination
  nodes [
   n1 = Sequence Player 'MistressSombre_Elimination_PACK'() {
    bLoopAnimation=false
    sequence=/BRCosmetics/Accessories/FORT_Backpacks/Backpack_F_MED_MistressSombre/Animation/MistressSombre_Elimination_PACK
   }
   Output Animation Pose(Result=@n1)
  ]
 }
 Transition {
  path=AnimGraph.AnimGraphNode_StateMachine_1.FE Idle / Elim.AnimStateTransitionNode_0.Transition
  nodes [Result(bCanEnterTransition=Get FE_ElimAnim() -> (FE_ElimAnim:bool))]
 }
 Transition {path=AnimGraph.AnimGraphNode_StateMachine_1.FE Idle / Elim.AnimStateTransitionNode_1.Transition, nodes [Result()]}
 Idle / Reactive {
  entry=Idle
  path=AnimGraph.AnimGraphNode_StateMachine_2.Idle / Reactive
  ids=id
  nodes [
   AnimStateNode_0 = Idle()
   AnimStateNode_1 = EnemyPlayerElimination()
   AnimStateNode_2 = ChestOpened() {is=title AnimStateNode}
   AnimStateTransitionNode_0 = AnimStateTransitionNode()
   AnimStateTransitionNode_1 = AnimStateTransitionNode() {bAutomaticRuleBasedOnSequencePlayerInState=true}
   AnimStateTransitionNode_2 = AnimStateTransitionNode()
   AnimStateTransitionNode_3 = AnimStateTransitionNode() {bAutomaticRuleBasedOnSequencePlayerInState=true}
   AnimStateTransitionNode_4 = AnimStateTransitionNode()
   AnimStateTransitionNode_5 = AnimStateTransitionNode()
  ]
  links [
   {
    dst=EnemyPlayerElimination
    rule=AnimGraph.AnimGraphNode_StateMachine_2.Idle / Reactive.AnimStateTransitionNode_5.Transition
    src=ChestOpened
    via=AnimStateTransitionNode_5
   }
   {
    dst=Idle
    rule=AnimGraph.AnimGraphNode_StateMachine_2.Idle / Reactive.AnimStateTransitionNode_3.Transition
    src=ChestOpened
    via=AnimStateTransitionNode_3
   }
   {
    dst=ChestOpened
    rule=AnimGraph.AnimGraphNode_StateMachine_2.Idle / Reactive.AnimStateTransitionNode_4.Transition
    src=EnemyPlayerElimination
    via=AnimStateTransitionNode_4
   }
   {
    dst=Idle
    rule=AnimGraph.AnimGraphNode_StateMachine_2.Idle / Reactive.AnimStateTransitionNode_1.Transition
    src=EnemyPlayerElimination
    via=AnimStateTransitionNode_1
   }
   {
    dst=ChestOpened
    rule=AnimGraph.AnimGraphNode_StateMachine_2.Idle / Reactive.AnimStateTransitionNode_2.Transition
    src=Idle
    via=AnimStateTransitionNode_2
   }
   {
    dst=EnemyPlayerElimination
    rule=AnimGraph.AnimGraphNode_StateMachine_2.Idle / Reactive.AnimStateTransitionNode_0.Transition
    src=Idle
    via=AnimStateTransitionNode_0
   }
  ]
 }
 Idle {
  path=AnimGraph.AnimGraphNode_StateMachine_2.Idle / Reactive.AnimStateNode_0.Idle
  nodes [Output Animation Pose(Result=Sequence Player 'MistressSombre_Idle_PACK'() {sequence=&MistressSombre_Idle_PACK})]
 }
 EnemyPlayerElimination {
  path=AnimGraph.AnimGraphNode_StateMachine_2.Idle / Reactive.AnimStateNode_1.EnemyPlayerElimination
  nodes [
   n1 = Sequence Player 'MistressSombre_Elimination_PACK'() {
    bLoopAnimation=false
    sequence=&MistressSombre_Elimination_PACK
   }
   Output Animation Pose(Result=@n1)
  ]
 }
 ChestOpened {
  path=AnimGraph.AnimGraphNode_StateMachine_2.Idle / Reactive.AnimStateNode_2.ChestOpened
  nodes [
   n1 = Sequence Player 'MistressSombre_ChestOpen_PACK'() {bLoopAnimation=false, sequence=/BRCosmetics/Accessories/FORT_Backpacks/Backpack_F_MED_MistressSombre/Animation/MistressSombre_ChestOpen_PACK}
   Output Animation Pose(Result=@n1)
  ]
 }
 Transition {
  path=AnimGraph.AnimGraphNode_StateMachine_2.Idle / Reactive.AnimStateTransitionNode_0.Transition
  nodes [Result(bCanEnterTransition=Get bEnemyPlayerEliminated() -> (bEnemyPlayerEliminated:bool))]
 }
 Transition {path=AnimGraph.AnimGraphNode_StateMachine_2.Idle / Reactive.AnimStateTransitionNode_1.Transition, nodes [Result()]}
 Transition {
  path=AnimGraph.AnimGraphNode_StateMachine_2.Idle / Reactive.AnimStateTransitionNode_2.Transition
  nodes [Result(bCanEnterTransition=Get ChestOpened() -> (ChestOpened:bool))]
 }
 Transition {path=AnimGraph.AnimGraphNode_StateMachine_2.Idle / Reactive.AnimStateTransitionNode_3.Transition, nodes [Result()]}
 Transition {
  path=AnimGraph.AnimGraphNode_StateMachine_2.Idle / Reactive.AnimStateTransitionNode_4.Transition
  nodes [Result(bCanEnterTransition=Get ChestOpened())]
 }
 Transition {
  path=AnimGraph.AnimGraphNode_StateMachine_2.Idle / Reactive.AnimStateTransitionNode_5.Transition
  nodes [Result(bCanEnterTransition=Get bEnemyPlayerEliminated())]
 }
 EventGraph {nodes [
   Event BlueprintUpdateAnimation()
   > n1 = Sequence!(execute) -> (then_0, then_1, then_2:Exec, then_3:Exec)
   n2 = Branch(execute=@n1.then_3, Condition=Get bIsInFrontEnd())
   > Set DeltaAlpha(DeltaAlpha:double=1) -> (Output_Get:double)
   Set DeltaAlpha(execute=@n2.else, DeltaAlpha=1)
   n3 = Branch(execute=@n1.then_1, Condition=Get bIsSkydiveDiveMode())
   n4 = GetOwnerSkeletalMeshComponent()
   n5 = BreakRotator(InRot=BreakTransform(InTransform=GetDeltaTransformFromRefPose(self=@n4, BoneName=spine_05, BaseName=root)).Rotation)
   n6 = ClampAxis!(Angle:float=@n5.Yaw) -> (ReturnValue:float)
   > Set Gimbal Rot(Gimbal Rot:Rotator=MakeRotator(Roll=float * float(A=@n6, B=-0.5) {is=title K2Node_PromotableOperator}, Pitch=float * float(A=@n6, B=-0.8))) -> (Output_Get:Rotator)
   Set Gimbal Rot(execute=@n3.else, Gimbal Rot=MakeRotator(Roll=float * float(A=@n5.Yaw, B=-0.5), Pitch=float * float(A=@n5.Yaw, B=-0.8)))
   > Branch(Condition=NOT Boolean(A=Get bIsCrouching()))
   > Set ADSGimbal(ADSGimbal:Rotator=MakeRotator(Roll=float * float(A=@n6, B=-1), Pitch=float * float(A=@n6, B=0.3))) -> (Output_Get:Rotator)
   > Set Is Jumping Float(Is Jumping Float:double="To Float (Boolean)"(InBool=Get bIsSkyDiving())) -> (Output_Get:double)
   > Set Is Floating(Is Floating:double="To Float (Boolean)"(InBool=Get bIsMoving2D())) -> (Output_Get:double)
   Branch(execute=@n1.then_2, Condition=NOT Boolean(A=Get bIsSkyDiving()))
   > Set Rot Down(Rot Down:float=GetAlphaFromReferenceTransformDelta(
     SkeletalMeshComponent="Get Skel mesh "() -> ("Skel mesh ":Skeletal Mesh Component Object Reference)
     BoneName=spine_03
     BaseName="head "
     InputRange=(X=0,Y=-120)
     OutputRange=(X=0,Y=-1)
     TwistAxis="0,0,1"
    )) -> (Output_Get:float)
   > Set Rot Up(Rot Up:float=GetAlphaFromReferenceTransformDelta(SkeletalMeshComponent="Get Skel mesh "(), BoneName=spine_04, BaseName=head, InputRange=(X=15,Y=60), TwistAxis="0,0,1")) -> (Output_Get:float)
   n7 = GetRealTimeSeconds()
   n8 = "Sin (Radians)"(A=float * float(A=float * float(A=1, B=1.4) {is=title K2Node_PromotableOperator}, B=@n7) {is=title K2Node_CommutativeAssociativeBinaryOperator})
   n9 = "Sin (Radians)"(A=float * float(A=float * float(A=1, B=0.8) {is=title K2Node_PromotableOperator}, B=@n7) {is=title K2Node_CommutativeAssociativeBinaryOperator})
   n10 = "Sin (Radians)"(A=float * float(A=float * float(A=1, B=2) {is=title K2Node_PromotableOperator}, B=@n7) {is=title K2Node_CommutativeAssociativeBinaryOperator})
   > Set Jar Rot(Jar Rot:Rotator=MakeRotator(Roll=float / float(A=@n8, B=-0.25), Pitch=float / float(A=@n10, B=-0.5), Yaw=float / float(A=@n9, B=-0.75))) -> (Output_Get:Rotator)
   n11 = Get bIsSkydiveDiveMode()
   > Set AlphaGimbal(AlphaGimbal:double="To Float (Boolean)"(InBool=NOT Boolean(A=@n11))) -> (Output_Get:double)
   n12 = Gravity Override State Function DEPRECATED(
    execute=@n1.then_0
    CharacterPartAnimInstance=Self-Reference()
    GameTimeInSeconds=GetGameTimeInSeconds()
    GravityOverrideParams=Get Gravity Override Params()
   )
   > Set Gravity Override(Gravity Override=@n12)
   Event BlueprintPostEvaluateAnimation() {is=title K2Node_Event}
   > n13 = Sequence!(execute) -> (then_0, then_1)
   > Set bEnemyPlayerEliminated(bEnemyPlayerEliminated:bool=false) -> (Output_Get:bool)
   Set ChestOpened(execute=@n13.then_1, ChestOpened:bool=false) -> (Output_Get:bool)
   Event OnCharacterCustomizationLoaded()
   > n14 = Cast To FortPlayerPawn(Object=TryGetFortPawn())
   > n15 = Cast To B_Athena_PartModifier_MistressSombre!(
    execute
    Object=GetModifierForPartType!(self:Fort Player Pawn Object Reference=@n14, PartType:EFortCustomPartType Enum=Backpack) -> (ReturnValue:Custom Character Part Modifier Object Reference)
   ) -> (then, CastFailed, AsB Athena Part Modifier Mistress Sombre:B Athena Part Modifier Mistress Sombre Object Reference) {is=title K2Node_DynamicCast}
   > Set Reactivity Enabled(Reactivity Enabled:bool=Get Reactivity Enabled!(self:B Athena Part Modifier Mistress Sombre Object Reference=@n15) -> (Reactivity Enabled:bool)) -> (Output_Get:bool)
   > Branch(Condition=OR Boolean(
     A=IsCosmeticObjectDisplayedAsPreview!(CosmeticObject:Object Reference=@n15) -> (ReturnValue:bool)
     B=IsInFrontEnd!(self:Custom Cosmetic Modifier Base Object Reference=@n15) -> (ReturnValue:bool)
    ))
   > n16 = Branch(Condition=Get Reactivity Enabled!() -> (Reactivity Enabled))
   > n17 = Do Once(Reset=@n18.then)
   Set FE_ElimAnim(execute=[@n20.then, @n17.Completed], FE_ElimAnim:bool=true) -> (Output_Get:bool)
   > Ball_Reactivity(self:B Athena Part Modifier Mistress Sombre Object Reference=@n15)
   n18 = Set FE_ElimAnim(execute=@n16.else, FE_ElimAnim=false)
   Event BlueprintInitializeAnimation()
   > n19 = Cast To FortPlayerPawn(Object=TryGetFortPawn())
   > Set As Fort Player Pawn(As Fort Player Pawn:Fort Player Pawn Object Reference=@n19) -> (Output_Get:Fort Player Pawn Object Reference)
   > "Set Skel mesh "("Skel mesh ":Skeletal Mesh Component Object Reference=GetSkeletalMeshForPartType(self=@n19)) -> (Output_Get:Skeletal Mesh Component Object Reference)
   AnimNotify_FE_ElimEntered() {is=title K2Node_Event}
   > Set FE_ElimAnim(FE_ElimAnim=false)
   > n20 = Delay(Duration=13)
   n21 = Event BlueprintProcessAnimInputEvent()
   > n22 = Sequence()
   > Branch(Condition=AND Boolean(A=IsNamed(self=@n21.AnimInputEvent, InEventName=aie.EnemyPlayerEliminated), B=Get Reactivity Enabled()))
   > Set bEnemyPlayerEliminated(bEnemyPlayerEliminated=true)
   > Delay(Duration=0.2)
   > Set bEnemyPlayerEliminated(bEnemyPlayerEliminated=false)
   Branch(execute=@n22.then_1, Condition=IsNamed(self=@n21.AnimInputEvent, InEventName=aie.ChestOpened))
   > Set ChestOpened(ChestOpened=true)
   > Delay(Duration=0.2)
   > Set ChestOpened(ChestOpened=false)
  ]}
]
variables [
 ADSGimbal {type=Rotator}
 AlphaGimbal {type=number}
 As Fort Player Pawn {type=FortPlayerPawn}
 bEnemyPlayerEliminated {type=boolean}
 ChestOpened {type=boolean}
 DeltaAlpha {type=number}
 FE_ElimAnim {type=boolean}
 Gimbal Rot {type=Rotator}
 Gravity Override {type=Vector}
 Gravity Override Params {type=GravityOverrideParamsStruct}
 Is Floating {type=number}
 Is Jumping Float {type=number}
 Jar Rot {type=Rotator}
 Reactive Sound Cue {type=SoundBase, default=/BRCosmetics/Sounds/Backblings/MistressSombre/IP_Backbling_MisstressSombre_Reactive_Cue}
 Reactivity Enabled {type=boolean}
 Rot Down {type=number}
 Rot Up {type=number}
 "Skel mesh " {type=SkeletalMeshComponent}
]
